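# Physics asset for the Canary Jungle male body cosmetic, covering its hanging bag, holster, rope and whip handle.
# Three chains branch from pelvis-linked master bones, with angular limits along each: a two-part bag, a holster, and a rope ending in a two-part whip handle.
# Capsules on pelvis, lower spine, thighs and left calf use free angular constraints and give body shapes around the accessories.
# A large unconstrained sphere on floor_col, offset downward, approximates a floor.
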PhysicsAsset /BRCosmetics/Characters/Player/Male/Medium/Bodies/M_MED_Canary_Jungle/Meshes/M_MED_Canary_Jungle_New_RopeBagPhysics
bodies {
 calf_l [{shapeType=Capsule, length=32.7, radius=9.6, shapeName=calf_l_capsule, center=<-32.0481 0 0>, rotation=<rot 90 0 0>}]
 dyn_bag_Main [{
   shapeType=Capsule
   length=10.25
   radius=4.75
   shapeName=dyn_bag_Main_capsule
   center=<6.2515 -1 2.5>
   rotation=<rot -4.6137 -89.2438 -15.4406>
  }]
 dyn_bag_Main2 [{shapeType=Capsule, length=12.75, radius=4.5, shapeName=dyn_bag_Main2_capsule, center=<5 -1 4>, rotation=<rot -5.7036 -90 -13.9008>}]
 dyn_holster [{shapeType=Capsule, length=14.0768, radius=6.6876, shapeName=dyn_holster_capsule, center=<5 1 5.25>, rotation=<rot 86.5771 179.9998 0>}]
 dyn_holster_MSTR [{shapeType=Capsule, length=0.505, radius=0.505, shapeName=dyn_holster_MSTR_capsule, rotation=<rot 0 -90 0>}]
 dyn_ropeMain_01 [{shapeType=Capsule, length=17.75, radius=6, shapeName=dyn_ropeMain_01_capsule, center=<6.2597 4 0>}]
 dyn_whip_handle [{shapeType=Capsule, length=13.5048, radius=2.4051, shapeName=dyn_whip_handle_capsule, center=<9.5 0 0>, rotation=<rot 90 0 0>}]
 dyn_whip_handle_01 [{shapeType=Capsule, length=4.3651, radius=3.1699, shapeName=dyn_whip_handle_01_capsule, center=<6 0 0>, rotation=<rot 90 0 0>}]
 floor_col [{shapeType=Sphere, radius=300, shapeName=floor_col_sphere, center=<0 0 -295>}]
 mstr_bag_Main [{shapeType=Capsule, length=0.505, radius=0.505, shapeName=mstr_bag_Main_capsule, rotation=<rot 0 -90 0>}]
 mstr_ropeMain [
  {shapeType=Capsule, length=20.8733, radius=4.5, shapeName=mstr_ropeMain_capsule, center=<5 5 1>}
  {shapeType=Capsule, length=5.25, radius=3.25, shapeName=mstr_ropeMain_capsule, center=<8.3086 0 12.8402>, rotation=<rot 90 0 -10>}
 ]
 mstr_ropeMain1 [{
   shapeType=Capsule
   length=2.3547
   radius=1.5406
   shapeName=mstr_ropeMain1_capsule
   center=<-0.2021 -0.3805 -0.1139>
   rotation=<rot 0 131.7785 -79.7277>
  }]
 pelvis [{shapeType=Capsule, length=3.75, radius=10.0799, shapeName=pelvis_capsule, center=<-9.9014 0.2808 4.5919>}]
 spine_01 [{shapeType=Capsule, length=7.755, radius=11.755, shapeName=spine_01_capsule, center=<-5.1465 -1.6771 0.4511>, rotation=<rot 0 -90 0>}]
 spine_02 [{
   shapeType=Capsule
   length=8.8703
   radius=11.6107
   shapeName=spine_02_capsule
   center=<-2.5624 -1.2115 0.2296>
   rotation=<rot 7.6949 -90 -0.7625>
  }]
 thigh_l [{shapeType=Capsule, length=22.6789, radius=9.8931, shapeName=thigh_l_capsule, center=<-18.0189 0.0728 0.4196>, rotation=<rot 90 0 0>}]
 thigh_r [{
   shapeType=Capsule
   length=26.653
   radius=9.5101
   shapeName=thigh_r_capsule
   center=<20.3245 -0.2026 -2.2027>
   rotation=<rot 90 89.2958 86.333>
  }]
 thigh_twist_01_l [{
   shapeType=Capsule
   length=23.4179
   radius=6.7304
   shapeName=thigh_twist_01_l_capsule
   center=<1.1502 0.931 4.4522>
   rotation=<rot 90 180 179.1282>
  }]
}
constraints [
 {
  bone1Name=thigh_l
  bone2Name=pelvis
  swing1LimitDegrees=45
  swing1Motion=Free
  swing2LimitDegrees=45
  swing2Motion=Free
  twistLimitDegrees=45
  twistMotion=Free
 }
 {
  bone1Name=thigh_twist_01_l
  bone2Name=thigh_l
  swing1LimitDegrees=45
  swing1Motion=Free
  swing2LimitDegrees=45
  swing2Motion=Free
  twistLimitDegrees=45
  twistMotion=Free
 }
 {
  bone1Name=dyn_holster_MSTR
  bone2Name=pelvis
  swing1LimitDegrees=45
  swing1Motion=Free
  swing2LimitDegrees=45
  swing2Motion=Free
  twistLimitDegrees=45
  twistMotion=Free
 }
 {bone1Name=dyn_holster, bone2Name=dyn_holster_MSTR, swing1LimitDegrees=25, swing2LimitDegrees=35, twistLimitDegrees=10}
 {
  bone1Name=thigh_r
  bone2Name=pelvis
  swing1LimitDegrees=45
  swing1Motion=Free
  swing2LimitDegrees=45
  swing2Motion=Free
  twistLimitDegrees=45
  twistMotion=Free
 }
 {
  bone1Name=spine_01
  bone2Name=pelvis
  swing1LimitDegrees=45
  swing1Motion=Free
  swing2LimitDegrees=45
  swing2Motion=Free
  twistLimitDegrees=45
  twistMotion=Free
 }
 {
  bone1Name=mstr_bag_Main
  bone2Name=pelvis
  swing1LimitDegrees=45
  swing1Motion=Free
  swing2LimitDegrees=45
  swing2Motion=Free
  twistLimitDegrees=45
  twistMotion=Free
 }
 {bone1Name=dyn_bag_Main, bone2Name=mstr_bag_Main, swing1LimitDegrees=20, swing2LimitDegrees=10, twistLimitDegrees=5}
 {bone1Name=dyn_bag_Main2, bone2Name=dyn_bag_Main, swing1LimitDegrees=10, swing2LimitDegrees=10, twistLimitDegrees=5}
 {
  bone1Name=spine_02
  bone2Name=spine_01
  swing1LimitDegrees=45
  swing1Motion=Free
  swing2LimitDegrees=45
  swing2Motion=Free
  twistLimitDegrees=45
  twistMotion=Free
 }
 {
  bone1Name=mstr_ropeMain1
  bone2Name=pelvis
  swing1LimitDegrees=45
  swing1Motion=Free
  swing2LimitDegrees=45
  swing2Motion=Free
  twistLimitDegrees=45
  twistMotion=Free
 }
 {bone1Name=mstr_ropeMain, bone2Name=mstr_ropeMain1, swing1LimitDegrees=35, swing2LimitDegrees=30, twistLimitDegrees=15}
 {bone1Name=dyn_whip_handle_01, bone2Name=dyn_whip_handle, swing1LimitDegrees=25, swing2LimitDegrees=25, twistLimitDegrees=5}
 {bone1Name=dyn_whip_handle, bone2Name=dyn_ropeMain_01, swing1LimitDegrees=25, swing2LimitDegrees=25, twistLimitDegrees=5}
 {bone1Name=dyn_ropeMain_01, bone2Name=mstr_ropeMain, swing1LimitDegrees=30, swing2LimitDegrees=30, twistLimitDegrees=15}
 {
  bone1Name=calf_l
  bone2Name=thigh_l
  swing1LimitDegrees=45
  swing1Motion=Free
  swing2LimitDegrees=45
  swing2Motion=Free
  twistLimitDegrees=45
  twistMotion=Free
 }
]
