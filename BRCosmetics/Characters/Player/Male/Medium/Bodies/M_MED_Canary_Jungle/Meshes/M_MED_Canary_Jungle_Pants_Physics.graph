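# Physics asset for the pants mesh of the male medium Canary Jungle player cosmetic, covering dynamic pant bones and left leg bones.
# Capsules on thigh, thigh twist and calf form the leg, linked by free-swing, free-twist constraints.
# Pant capsules hang off these with tight angular limits: the upper from the thigh twist, the lower through a tiny MSTR anchor body that joins freely to the calf.
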PhysicsAsset /BRCosmetics/Characters/Player/Male/Medium/Bodies/M_MED_Canary_Jungle/Meshes/M_MED_Canary_Jungle_Pants_Physics
bodies {
 calf_l [{shapeType=Capsule, length=24.2756, radius=6.8418, shapeName=calf_l_capsule, center=<-15.3525 -2.8385 0>, rotation=<rot 90 0 0>}]
 dyn_pant_leg [{shapeType=Capsule, length=10.3542, radius=2.6418, shapeName=dyn_pant_leg_capsule, center=<13 0 2>, rotation=<rot 0 0 90>}]
 dyn_pant_leg_calf [{shapeType=Capsule, length=13.4581, radius=2.1115, shapeName=dyn_pant_leg_calf_capsule, center=<10 0 0>}]
 dyn_pant_leg_calf_MSTR [{shapeType=Capsule, length=0.505, radius=0.505, shapeName=dyn_pant_leg_calf_MSTR_capsule, rotation=<rot 0 -90 0>}]
 thigh_l [{shapeType=Capsule, length=24.4289, radius=8.1431, shapeName=thigh_l_capsule, center=<-18.0189 0.0728 0>, rotation=<rot 90 0 0>}]
 thigh_twist_01_l [{
   shapeType=Capsule
   length=23.4179
   radius=6.7304
   shapeName=thigh_twist_01_l_capsule
   center=<-4.8518 -1.0343 0.5181>
   rotation=<rot 90 180 179.1282>
  }]
}
constraints [
 {bone1Name=dyn_pant_leg, bone2Name=thigh_twist_01_l, swing1LimitDegrees=5, swing2LimitDegrees=15, twistLimitDegrees=3}
 {
  bone1Name=thigh_twist_01_l
  bone2Name=thigh_l
  swing1LimitDegrees=45
  swing1Motion=Free
  swing2LimitDegrees=45
  swing2Motion=Free
  twistLimitDegrees=45
  twistMotion=Free
 }
 {
  bone1Name=dyn_pant_leg_calf_MSTR
  bone2Name=calf_l
  swing1LimitDegrees=45
  swing1Motion=Free
  swing2LimitDegrees=45
  swing2Motion=Free
  twistLimitDegrees=45
  twistMotion=Free
 }
 {bone1Name=dyn_pant_leg_calf, bone2Name=dyn_pant_leg_calf_MSTR, swing1LimitDegrees=15, swing2LimitDegrees=5, twistLimitDegrees=3}
 {
  bone1Name=calf_l
  bone2Name=thigh_l
  swing1LimitDegrees=45
  swing1Motion=Free
  swing2LimitDegrees=45
  swing2Motion=Free
  twistLimitDegrees=45
  twistMotion=Free
 }
]
