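# Physics asset of collision volumes for the cape of the M_MED_Canary_Jungle player body, as its CapeColliders name indicates.
# Overlapping capsules wrap the pelvis, spine, neck and clavicles, with a single sphere for the head.
# Further capsules cover accessory bones for the bag, rope and whip handle.
# Arm and leg bodies carry no shapes, and there are no constraints, so it defines colliders rather than a jointed ragdoll.
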PhysicsAsset /BRCosmetics/Characters/Player/Male/Medium/Bodies/M_MED_Canary_Jungle/Meshes/M_MED_Canary_Jungle_CapeColliders
bodies {
 calf_l []
 calf_r []
 clavicle_l [{shapeType=Capsule, length=7.7636, radius=11.1427, center=<9.9642 -0.7741 -1.5923>, rotation=<rot 75.9352 -0.0001 175.5622>}]
 clavicle_r [{shapeType=Capsule, length=7.7636, radius=11.1427, center=<-9.9642 0.7741 1.5923>, rotation=<rot 75.9351 -0.0001 -4.4378>}]
 dyn_bag_Main [{
   shapeType=Capsule
   length=16.7463
   radius=6.394
   shapeName=dyn_bag_Main_capsule
   center=<2.5754 -2.0504 1.707>
   rotation=<rot -2.934 -90 -15.472>
  }]
 dyn_bag_Main2 [{
   shapeType=Capsule
   length=15.3601
   radius=6.5752
   shapeName=dyn_bag_Main2_capsule
   center=<4.7365 0.8954 3.6746>
   rotation=<rot -12.4935 -90 -21.8743>
  }]
 dyn_ropeMain_01 [{
   shapeType=Capsule
   length=12.5279
   radius=5.9308
   shapeName=dyn_ropeMain_01_capsule
   center=<4.7182 -2.8766 -0.3281>
   rotation=<rot -1.1308 -90 2.7028>
  }]
 dyn_whip_handle [{
   shapeType=Capsule
   length=13.6836
   radius=3.1553
   shapeName=dyn_whip_handle_capsule
   center=<9.0941 0.2862 0.2353>
   rotation=<rot 0 -89.5039 -89.3915>
  }]
 dyn_whip_handle_01 [{
   shapeType=Capsule
   length=6.1151
   radius=3.6699
   shapeName=dyn_whip_handle_01_capsule
   center=<6.9476 1.854 0.067>
   rotation=<rot 0 98.3171 -89.5192>
  }]
 foot_l []
 foot_r []
 head [{shapeType=Sphere, radius=17.3125, center=<11.028 -0.174 0>}]
 lowerarm_l []
 lowerarm_r []
 mstr_ropeMain [{
   shapeType=Capsule
   length=11.0172
   radius=6.1648
   shapeName=mstr_ropeMain_capsule
   center=<5.72 -2.311 0.9218>
   rotation=<rot 0.9739 -89.5375 5.3012>
  }]
 neck_02 [{shapeType=Capsule, length=13.4375, radius=7.3175, center=<-0.2642 -0.3408 0>, rotation=<rot 90 -171.1193 -175.3978>}]
 pelvis [{shapeType=Capsule, length=8.6562, radius=13.5388, center=<-2.6215 -1.3695 0>}]
 spine_01 [{shapeType=Capsule, length=10.7187, radius=12.4138, center=<-2.6215 -0.7156 0>}]
 spine_02 [
  {shapeType=Capsule, length=10.5937, radius=10.945, center=<-2.6215 0.4993 0.0001>}
  {shapeType=Capsule, length=29.875, radius=11.1013, center=<2.5024 0.4992 3.9825>, rotation=<rot 90 0.0049 0.0049>}
  {shapeType=Capsule, length=29.875, radius=11.1013, center=<2.4704 0.4991 -3.9825>, rotation=<rot 90 -0.0396 -0.0396>}
 ]
 spine_03 [{shapeType=Capsule, length=9.1562, radius=10.7575, center=<-1.953 0.1364 0.0001>}]
 spine_04 [{shapeType=Capsule, length=14.6562, radius=13.6325, center=<2.1467 -1.2748 0.0002>}]
 spine_05 [
  {shapeType=Capsule, length=17.8437, radius=17.195, center=<-1.7514 2.4879 0>}
  {shapeType=Capsule, length=12.7812, radius=17.0625, center=<0.7875 3.1143 0>}
 ]
 thigh_l []
 thigh_r []
 upperarm_l []
 upperarm_r []
}
constraints []
